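# Niagara idle effect for the YogaPatio pickaxe that plays a sheen followed by a sparkle, suppressed while User.bisSwinging is true.
# One emitter fires two delayed bursts, the second timed just before the sheen ends, using spawn groups to vary lifetime, offset and color.
# Spawn group also sets a visibility tag for the mesh and sprite renderers.
# Curves fade particles and animate a material parameter, while time-of-day values tint and scale brightness.
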NiagaraSystem /BRCosmetics/Weapons/FORT_Melee/Pickaxe_YogaPatio/FX/NS_Pickaxe_YogaPatio_Idle
dependencies {
 usedDataInterfaces [&NiagaraDataInterfaceCurve, &NiagaraDataInterfaceColorCurve]
 usedDynamicInputs [&ColorFromCurve, &FNTimeOfDayScalar]
 usedModules [
  &self:SystemSpawnScript.SystemScriptSource.SystemScriptGraph.NiagaraNodeAssignment_0.SetVariables_#1
  &SystemState
  &self:SystemSpawnScript.SystemScriptSource.SystemScriptGraph.NiagaraNodeAssignment_1.SetVariables_#2
 ]
}
emitters {SheenSparkle_FX {
  emitterUpdateScript [
   &EmitterState()
   &SpawnBurst_Instantaneous(
    Spawn Probability=&MakeCustomFloatFromBool(Bool=$User.bisSwinging, True Float=0, False Float=1)
    Spawn Time=$System.Sheen_Delay
    Use Spawn Probability=true
    UseLoopCountLimit=false
   ) {module=SpawnBurst_Instantaneous001}
   &SpawnBurst_Instantaneous(
    Spawn Group=1
    Spawn Probability=&MakeCustomFloatFromBool(Bool=$User.bisSwinging, True Float=0, False Float=1)
    Spawn Time=&Add_Float(A=$System.Sheen_Lifetime, B=&Add_Float(A=-0.1, B=$System.Sheen_Delay))
    Use Spawn Probability=true
    UseLoopCountLimit=false
   ) {module=SpawnBurst_Instantaneous002}
  ]
  particleSpawnScript [
   /SetVariables(Particles.VisibilityTag=&SpawnGroup()) {isSetParameters=true}
   &InitializeParticle(
    Color=&MaskLinearColorBySpawnGroup(
     Masked Linear Color=$System.TODColor
     Passthrough Linear Color=&MultiplyLinearColor(Linear Color=<rgba 5 5 5 1>, Scale Factor=&MakeVector4_FromLinearColor(LinearColor=$System.TODColor))
    )
    Lifetime=&MaskFloatBySpawnGroup(Masked Float=$System.Sheen_Lifetime, Passthrough Float=&RandomRangeFloat(Minimum=0.4, Maximum=0.6))
    Mesh Scale Mode=::NewEnumerator3
    Position Offset=&MaskVectorBySpawnGroup(Masked Vector=<0 0 0>, Passthrough Vector=<0 0 75.5>)
    Sprite Rotation Mode=::NewEnumerator2
    Sprite Size Mode=::NewEnumerator1
    Uniform Sprite Size Max=31
    Uniform Sprite Size Min=27
    UsePositionOffset=true
    Write Material Random=true
   ) {module=InitializeParticle001}
   &CameraOffset(Camera Offset Amount=&MaskFloatBySpawnGroup(Masked Float=0.05, Passthrough Float=10))
  ]
  particleUpdateScript [
   &ParticleState()
   &ScaleSpriteSize(
    Uniform Curve Scale=1.2
    Uniform Curve Sprite Scale=DataInterface() {propertyValues {Curve {keys [
        {arriveTangent=9.0725, interpMode=RCIM_Cubic, leaveTangent=9.0725, tangentMode=RCTM_User, time=0, value=0}
        {arriveTangent=-1.7917, interpMode=RCIM_Cubic, leaveTangent=-1.7917, tangentMode=RCTM_User, time=0.0666, value=0.9847}
        {arriveTangent=-0.0405, interpMode=RCIM_Cubic, leaveTangent=-0.0405, tangentMode=RCTM_User, time=1, value=0}
       ]}}}
   ) {module=ScaleSpriteSize001}
   &SpriteRotationRate(Rotation Rate=&RandomRangeFloat(Minimum=90, Maximum=360))
   &ScaleColor(
    Scale Alpha=&FloatFromCurve(FloatCurve=DataInterface() {propertyValues {Curve {keys [
         {arriveTangent=3.6745, interpMode=RCIM_Cubic, leaveTangent=3.6745, tangentMode=RCTM_User, time=0, value=0}
         <at 0.5016 0.9767 RCIM_Cubic>
         {arriveTangent=-1.1758, interpMode=RCIM_Cubic, leaveTangent=-1.1758, tangentMode=RCTM_User, time=1, value=0}
        ]}}})
    Scale RGB=&VectorFromFloat(Value=&FloatFromCurve(FloatCurve=DataInterface() {propertyValues {Curve {keys [
          {arriveTangent=4.2309, interpMode=RCIM_Cubic, leaveTangent=4.2309, tangentMode=RCTM_User, time=0, value=0}
          <at 0.5017 1.0042 RCIM_Cubic>
          {arriveTangent=-3.8662, tangentMode=RCTM_User, time=1, value=0}
         ]}}}))
    ScaleA=true
    ScaleRGB=true
   ) {module=ScaleColor001}
   &ScaleColor(Scale RGB=&VectorFromFloat(Value=$System.TODFloat), ScaleA=true, ScaleRGB=true)
   &DynamicMaterialParameters(
    Index 0 Param 1=&FloatFromCurve(FloatCurve=DataInterface() {propertyValues {Curve {keys [{interpMode=RCIM_Cubic, tangentMode=RCTM_User, time=0, value=0}, {interpMode=RCIM_Cubic, tangentMode=RCTM_User, time=1, value=1}]}}})
    Index 0 Param 2=0.1
    Index 0 Param 3=0.12
    Module.Index0Param0Write=true
    Module.Index0Param1Write=true
    Module.Index0Param2Write=true
    Module.Index0Param3Write=true
   )
  ]
  rendererClasses [&NiagaraMeshRendererProperties, &NiagaraSpriteRendererProperties]
  renderers [
   {rendererClass=&NiagaraMeshRendererProperties, rendererIndex=0}
   {rendererClass=&NiagaraSpriteRendererProperties, rendererIndex=1}
  ]
 }}
summary {
 systemName=NS_Pickaxe_YogaPatio_Idle
 emitters [{
   emitterName=SheenSparkle_FX
   rendererClasses [&NiagaraMeshRendererProperties, &NiagaraSpriteRendererProperties]
  }]
 userVariables [User.bisSwinging {type=NiagaraBool}]
}
